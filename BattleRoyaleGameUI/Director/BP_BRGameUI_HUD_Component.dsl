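; HUD director component that adds Battle Royale HUD scenes when UI is ready and removes them on EndPlay.
; HUD settings and CVars gate the default, quickbar, text chat, squad list, kill feed and hit tick scenes.
; Subsystem CVar delegates toggle scenes live, while HUD visibility and inventory custom-UI changes re-evaluate HUD and quickbar.
; A gameplay event listener toggles the chat-open scene on compact UI.
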(event EventOnReadyToShowUI
  (bind _returnvalue (CallFunction|GetOwningLocalPlayerController))
  (bind _returnvalue_2 (Actor|GetComponentbyClass _returnvalue "/Script/ItemizationCoreRuntime.ControllerComponent_ItemizationInventory"))
  (bind _returnvalue_3 (Inventory|GetRootInventory _returnvalue_2))
  (CallFunction|AddDefaultHUDScenes)
  (EventDispatchers|AssignOnSquadListPlacementCVarChanged (GameInstanceSubsystems|GetBattleRoyaleGameUISubsystem) (EventDispatchers|CreateEvent :!event OnSquadListPlacementCVarChanged_Event))
  (EventDispatchers|BindEventtoOnNewHudEnabledCVarChanged (GameInstanceSubsystems|GetBattleRoyaleGameUISubsystem) (EventDispatchers|CreateEvent :!event OnNewHudCVarChanged_Event))
  (EventDispatchers|BindEventtoOnNewQuickbarEnabledCVarChanged (GameInstanceSubsystems|GetBattleRoyaleGameUISubsystem) (EventDispatchers|CreateEvent :!event OnNewQuickbarCVarChanged_Event))
  (EventDispatchers|BindEventtoOnKillFeedPlacementCVarChanged (GameInstanceSubsystems|GetBattleRoyaleGameUISubsystem) (EventDispatchers|CreateEvent :!event OnKillFeedPlacementCVarChanged_Event))
  (bind _asfort_player_controller_gameplay (Utilities|Casting|CastToFortPlayerControllerGameplay _returnvalue))
  (Utilities|IsValid _returnvalue
    (:Is%20Valid
      (HUD|BindEventtoOnHUDElementVisibilityRefreshedDelegate _asfort_player_controller_gameplay (EventDispatchers|CreateEvent :!event OnHUDElementVisibilityRefreshed))
      (Utilities|IsValid _returnvalue_2
        (:Is%20Valid
          (Utilities|IsValid _returnvalue_3
            (:Is%20Valid
              (bind _asfort_root_itemization_inventory (Utilities|Casting|CastToFortRootItemizationInventory _returnvalue_3))
              (Utilities|IsValid _asfort_root_itemization_inventory
                (:Is%20Valid
                  (Itemization|BindEventtoCustomUIChangedDynamicEvent _asfort_root_itemization_inventory (EventDispatchers|CreateEvent :!event BP_BRGameUI_HUD_Component_AutoGenFunc)))))))))))

(event EventEndPlay
  (params (EndPlayReason :type EEndPlayReason%20Enum :default "Destroyed"))
  (CallFunction|RemoveDefaultHUDScenes)
  (EventDispatchers|UnbindEventfromOnSquadListPlacementCVarChanged (GameInstanceSubsystems|GetBattleRoyaleGameUISubsystem) (EventDispatchers|CreateEvent :!event OnSquadListPlacementCVarChanged_Event))
  (EventDispatchers|UnbindEventfromOnNewHudEnabledCVarChanged (GameInstanceSubsystems|GetBattleRoyaleGameUISubsystem) (EventDispatchers|CreateEvent :!event OnNewHudCVarChanged_Event))
  (EventDispatchers|UnbindEventfromOnNewQuickbarEnabledCVarChanged (GameInstanceSubsystems|GetBattleRoyaleGameUISubsystem) (EventDispatchers|CreateEvent :!event OnNewQuickbarCVarChanged_Event))
  (EventDispatchers|UnbindEventfromOnKillFeedPlacementCVarChanged (GameInstanceSubsystems|GetBattleRoyaleGameUISubsystem) (EventDispatchers|CreateEvent :!event OnKillFeedPlacementCVarChanged_Event))
  (CallFunction|UnregisterChatActiveEvent))

(event Custom|RegisterChatActiveEvent
  (bind _returnvalue (CallFunction|GetOwningLocalPlayerController))
  (bind _chatopenscene (Variables|Default|GetChatOpenScene))
  (Utilities|IsValid _returnvalue
    (:Is%20Valid
      (bind (_asynctaskproxy _payload) (GameplayEventRouter|StartListeningToEvent (GameplayEventRouter|GetGlobalEventRouter) "/FortniteCoreUI/UI/Social/Chat/TextChat/BPS_TextChatActiveStateChange.BPS_TextChatActiveStateChange" _returnvalue)
        (:then
          (Variables|Default|SetChatActiveHandle _asynctaskproxy))
        (:OnEventReceived
          (if (DisplayHelper|IsUsingCompactUI)
            (if (Utilities|Struct|BreakBPSTextChatActiveStateChange _payload)
              (CallFunction|AddScene _chatopenscene)
              (else
                (CallFunction|RemoveScene _chatopenscene)))))))))

(event Custom|UnregisterChatActiveEvent
  (GameplayEventRouter|StopListeningtoEvent (Variables|Default|GetChatActiveHandle)))

(fn AddDefaultHUDScenes
  (Utilities|FlowControl|Sequence
    (:then_0
      (bind _isnewhud (Class|BPFLHUDSettings|IsNewHUDEnabled self))
      (if _isnewhud
        (CallFunction|AddScenes (Variables|Default|GetDefaultHUDScenes))
        (CallFunction|AddScene (select (DisplayHelper|IsUsingCompactUI) "/BattleRoyaleGameUI/Scenes/UIScene_BRGameUI_MobileTextChat.UIScene_BRGameUI_MobileTextChat" "/BattleRoyaleGameUI/Scenes/UIScene_BRGameUI_TextChat.UIScene_BRGameUI_TextChat"))
        (CallFunction|RegisterChatActiveEvent)))
    (:then_1
      (bind _isenabled (Class|BPFLHUDSettings|IsNewQuickbarEnabled self))
      (if _isenabled
        (CallFunction|AddScene (Variables|Default|GetQuickbarHudScene))))
    (:then_2
      (CallFunction|AddSquadListBottomScenes))
    (:then_3
      (CallFunction|AddKillFeedScene))
    (:then_4
      (bind _returnvalue (Development|GetConsoleVariableBoolValue "BattleRoyaleGameUI.NewHitTicks.Enabled"))
      (if _returnvalue
        (CallFunction|AddScene (Variables|Default|GetHitTicksUIScene))))))

(fn AddKillFeedScene
  (bind _returnvalue (Development|GetConsoleVariableBoolValue "BattleRoyalGameUI.KillFeedPositionOnRight.Enabled"))
  (if _returnvalue
    (bind _isnewhud (Class|BPFLHUDSettings|IsNewHUDEnabled self))
    (if _isnewhud
      (CallFunction|AddScene (Variables|Default|GetKillFeedScene)))))

(fn AddSquadListBottomScenes
  ;[1 1 1] Enabling scenes that move the squad health list from the top left to the bottom left
  (bind _returnvalue (Development|GetConsoleVariableBoolValue "BattleRoyalGameUI.SquadListPositionAtBottom.Enabled")) ; This is a cvar defined in BattleRoyaleGameUIModule.cpp that controls the position of the squad list
  (if _returnvalue
    (bind _isnewhud (Class|BPFLHUDSettings|IsNewHUDEnabled self))
    (if _isnewhud
      (CallFunction|AddScenes (Variables|Default|GetSquadListBottomScenes)))))

(fn BP_BRGameUI_HUD_Component_AutoGenFunc
  (params (Inventory :type Fort%20Root%20Itemization%20Inventory%20Object%20Reference))
  (CallFunction|OnHUDElementVisibilityRefreshed))

(fn OnHUDElementVisibilityRefreshed
  (bind _isnewhud (Class|BPFLHUDSettings|IsNewHUDEnabled self))
  (CallFunction|OnNewHudCVarChangedEvent _isnewhud)
  (bind _isenabled (Class|BPFLHUDSettings|IsNewQuickbarEnabled self))
  (CallFunction|OnNewQuickbarCVarChangedEvent _isenabled))

(fn OnKillFeedPlacementCVarChanged_Event
  (params (bNewValue :type Boolean))
  (if bNewValue
    (CallFunction|AddKillFeedScene)
    (else
      (Default|RemoveKillFeedScene))))

(fn OnNewHudCVarChanged_Event
  (params (bNewValue :type Boolean))
  (CallFunction|AddorRemoveScenes (Variables|Default|GetDefaultHUDScenes) bNewValue))

(fn OnNewQuickbarCVarChanged_Event
  (params (bNewValue :type Boolean))
  (CallFunction|AddorRemoveScene (Variables|Default|GetQuickbarHudScene) bNewValue))

(fn OnSquadListPlacementCVarChanged_Event
  (params (bSquadListIsBottom :type Boolean))
  (if bSquadListIsBottom
    (CallFunction|AddSquadListBottomScenes)
    (else
      (CallFunction|RemoveSquadListBottomScenes))))

(fn RemoveDefaultHUDScenes
  (Utilities|FlowControl|Sequence
    (:then_0
      (bind _isnewhud (Class|BPFLHUDSettings|IsNewHUDEnabled self))
      (if _isnewhud
        (CallFunction|RemoveScenes (Variables|Default|GetDefaultHUDScenes))))
    (:then_1
      (bind _isenabled (Class|BPFLHUDSettings|IsNewQuickbarEnabled self))
      (if _isenabled
        (CallFunction|RemoveScene (Variables|Default|GetQuickbarHudScene))))
    (:then_2
      (CallFunction|RemoveSquadListBottomScenes))
    (:then_3
      (Default|RemoveKillFeedScene))
    (:then_4
      (Default|RemoveHitTicksScene))))

(fn RemoveHitTicksScene
  (bind _returnvalue (Development|GetConsoleVariableBoolValue "BattleRoyaleGameUI.NewHitTicks.Enabled"))
  (if _returnvalue
    (CallFunction|RemoveScene (Variables|Default|GetHitTicksUIScene))))

(fn RemoveKillFeedScene
  (CallFunction|RemoveScene (Variables|Default|GetKillFeedScene)))

(fn RemoveSquadListBottomScenes
  (CallFunction|RemoveScenes (Variables|Default|GetSquadListBottomScenes)))
