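; UI director Blueprint for the battle royale game UI that warms the font cache at startup.
; On BeginPlay it runs PrecacheFonts, which loops over the fonts returned by GetFontsToCache.
; Each font is passed to PrecacheFontCharacters along with the commonly used character set from the FontCache variables.
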(event EventBeginPlay
  (CallFunction|PrecacheFonts))

(fn PrecacheFonts
  (for _array_element (Variables|FontCache|GetFontsToCache)
    (FontCache|PrecacheFontCharacters (Variables|FontCache|GetCommonlyUsedCharacters) _array_element)))

(fn ConstructionScript)
